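# Physics asset for the F_MED_BluntWhimsy female player body cosmetic, covering only the collar and upper spine, apparently for secondary collar motion.
# Three capsule bodies form a chain from a large spine_05 volume through a small base_collar to dyn_collar.
# The base_collar joint to spine_05 is free on all angular axes, while dyn_collar acts as a narrow limited hinge on base_collar, with its other swing and twist locked.
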PhysicsAsset /BRCosmetics/Characters/Player/Female/Medium/Bodies/F_MED_BluntWhimsy/Meshes/F_MED_BluntWhimsy_Physics
bodies {
 base_collar [{shapeType=Capsule, length=0.505, radius=0.505, shapeName=base_collar_capsule, rotation=<rot 0 -90 0>}]
 dyn_collar [{shapeType=Capsule, length=4, radius=1, shapeName=dyn_collar_capsule, center=<2 0 0>, rotation=<rot 90 0 0>}]
 spine_05 [{shapeType=Capsule, length=10.1656, radius=10, shapeName=spine_05_capsule, center=<-3.2417 0.6362 0>, rotation=<rot 90 -165.9638 -125.364>}]
}
constraints [
 {
  bone1Name=base_collar
  bone2Name=spine_05
  swing1LimitDegrees=45
  swing1Motion=Free
  swing2LimitDegrees=45
  swing2Motion=Free
  twistLimitDegrees=45
  twistMotion=Free
 }
 {
  bone1Name=dyn_collar
  bone2Name=base_collar
  swing1LimitDegrees=20
  swing2LimitDegrees=45
  swing2Motion=Locked
  twistLimitDegrees=45
  twistMotion=Locked
 }
]
