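# Physics asset for the BluntWhimsy female medium cosmetic's FaceAcc mesh, covering bangs, side and back hair, and earrings.
# Small base capsules link to neck_02 with free angular constraints and anchor one-segment bang and earring bodies and two-segment hair chains.
# The dyn bodies have narrow swing limits, with twist zero-limited or locked.
# Neck, clavicle and upper-arm capsules supply body shapes around the hair.
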PhysicsAsset /BRCosmetics/Characters/Player/Female/Medium/Bodies/F_MED_BluntWhimsy/Meshes/Parts/F_MED_BluntWhimsy_FaceAcc_Physics
bodies {
 base_bang_c_1 [{shapeType=Capsule, length=0.505, radius=0.505, shapeName=base_bang_c_1_capsule, rotation=<rot 0 -90 0>}]
 base_bang_l_1 [{shapeType=Capsule, length=0.505, radius=0.505, shapeName=base_bang_l_1_capsule, rotation=<rot 0 -90 0>}]
 base_bang_r_1 [{shapeType=Capsule, length=0.505, radius=0.505, shapeName=base_bang_r_1_capsule, rotation=<rot 0 -90 0>}]
 base_earring_l [{shapeType=Capsule, length=0.505, radius=0.505, shapeName=base_earring_l_capsule, rotation=<rot 0 -90 0>}]
 base_earring_r [{shapeType=Capsule, length=0.505, radius=0.505, shapeName=base_earring_r_capsule, rotation=<rot 0 -90 0>}]
 base_hair_A_l_1 [{shapeType=Capsule, length=0.505, radius=0.505, shapeName=base_hair_A_l_1_capsule, rotation=<rot 0 -90 0>}]
 base_hair_A_r_1 [{shapeType=Capsule, length=0.505, radius=0.505, shapeName=base_hair_A_r_1_capsule, rotation=<rot 0 -90 0>}]
 base_hair_B_l_1 [{shapeType=Capsule, length=0.505, radius=0.505, shapeName=base_hair_B_l_1_capsule, rotation=<rot 0 -90 0>}]
 base_hair_B_r_1 [{shapeType=Capsule, length=0.505, radius=0.505, shapeName=base_hair_B_r_1_capsule, rotation=<rot 0 -90 0>}]
 base_hair_Bk_l_1 [{shapeType=Capsule, length=0.505, radius=0.505, shapeName=base_hair_Bk_l_1_capsule, rotation=<rot 0 -90 0>}]
 base_hair_Bk_r_1 [{shapeType=Capsule, length=0.505, radius=0.505, shapeName=base_hair_Bk_r_1_capsule, rotation=<rot 0 -90 0>}]
 clavicle_l [{shapeType=Capsule, length=2, radius=5, shapeName=clavicle_l_capsule, center=<5 0 -1.7995>, rotation=<rot 90 0 0>}]
 clavicle_r [{shapeType=Capsule, length=2, radius=5, shapeName=clavicle_r_capsule, center=<-5 1.8672 1.9077>, rotation=<rot 90 0 0>}]
 dyn_bang_c_1 [{shapeType=Capsule, length=4, radius=2, shapeName=dyn_bang_c_1_capsule, center=<3 0 0>, rotation=<rot 90 0 0>}]
 dyn_bang_l_1 [{shapeType=Capsule, length=4, radius=2, shapeName=dyn_bang_l_1_capsule, center=<3 0 0>, rotation=<rot 90 0 0>}]
 dyn_bang_r_1 [{shapeType=Capsule, length=4, radius=2, shapeName=dyn_bang_r_1_capsule, center=<3 0 0>, rotation=<rot 90 0 0>}]
 dyn_earring_l [{shapeType=Capsule, length=2, radius=1, shapeName=dyn_earring_l_capsule, center=<2 0 0>, rotation=<rot 90 0 0>}]
 dyn_earring_r [{shapeType=Capsule, length=2, radius=1, shapeName=dyn_earring_r_capsule, center=<2 0 0>, rotation=<rot 90 0 0>}]
 dyn_hair_A_l_1 [{shapeType=Capsule, length=1, radius=4, shapeName=dyn_hair_A_l_1_capsule, center=<3 0 0>, rotation=<rot 90 0 0>}]
 dyn_hair_A_l_2 [{shapeType=Capsule, length=4, radius=4, shapeName=dyn_hair_A_l_2_capsule, center=<4.9486 0 1.1203>, rotation=<rot 0 0 90>}]
 dyn_hair_A_r_1 [{shapeType=Capsule, length=1, radius=4, shapeName=dyn_hair_A_r_1_capsule, center=<3 0 0>, rotation=<rot 90 0 0>}]
 dyn_hair_A_r_2 [{shapeType=Capsule, length=6, radius=2, shapeName=dyn_hair_A_r_2_capsule, center=<9.1815 -0.2502 -2.8379>, rotation=<rot 0 180 80>}]
 dyn_hair_B_l_1 [{shapeType=Capsule, length=1, radius=4, shapeName=dyn_hair_B_l_1_capsule, center=<3 0 0>, rotation=<rot 90 0 0>}]
 dyn_hair_B_l_2 [{shapeType=Capsule, length=4, radius=4, shapeName=dyn_hair_B_l_2_capsule, center=<3.2185 0 0.6963>, rotation=<rot 0 0 90>}]
 dyn_hair_B_r_1 [{shapeType=Capsule, length=1, radius=4, shapeName=dyn_hair_B_r_1_capsule, center=<3 0 0>, rotation=<rot 90 0 0>}]
 dyn_hair_B_r_2 [{shapeType=Capsule, length=6, radius=2, shapeName=dyn_hair_B_r_2_capsule, center=<5.1805 0 -0.8581>, rotation=<rot 0 -0.0001 93.1213>}]
 dyn_hair_Bk_l_1 [{shapeType=Capsule, length=1, radius=4, shapeName=dyn_hair_Bk_l_1_capsule, center=<3 0 0>, rotation=<rot 90 0 0>}]
 dyn_hair_Bk_l_2 [{shapeType=Capsule, length=4, radius=3.5, shapeName=dyn_hair_Bk_l_2_capsule, center=<4.485 0 -0.5381>, rotation=<rot 0 0 90>}]
 dyn_hair_Bk_r_1 [{shapeType=Capsule, length=1, radius=4, shapeName=dyn_hair_Bk_r_1_capsule, center=<3 0 0>, rotation=<rot 90 0 0>}]
 dyn_hair_Bk_r_2 [{shapeType=Capsule, length=4, radius=3.5, shapeName=dyn_hair_Bk_r_2_capsule, center=<4.2566 0 -0.3642>, rotation=<rot 0 0 90>}]
 neck_01 [{shapeType=Capsule, length=5, radius=5, shapeName=neck_01_capsule, rotation=<rot 0 -90 0>}]
 neck_02 [{shapeType=Capsule, length=6, radius=6, shapeName=neck_02_capsule, rotation=<rot 90 0 0>}]
 upperarm_l [{shapeType=Capsule, length=14.5, radius=5, shapeName=upperarm_l_capsule, center=<8.5 -0.9 -1.8>, rotation=<rot 90 0 0>}]
 upperarm_r [{shapeType=Capsule, length=15, radius=4.5, shapeName=upperarm_r_capsule, center=<-11.2521 0.9484 0.7509>, rotation=<rot 90 0 0>}]
}
constraints [
 {
  bone1Name=neck_02
  bone2Name=neck_01
  swing1LimitDegrees=45
  swing1Motion=Free
  swing2LimitDegrees=45
  swing2Motion=Free
  twistLimitDegrees=45
  twistMotion=Free
 }
 {
  bone1Name=base_bang_c_1
  bone2Name=neck_02
  swing1LimitDegrees=45
  swing1Motion=Free
  swing2LimitDegrees=45
  swing2Motion=Free
  twistLimitDegrees=45
  twistMotion=Free
 }
 {bone1Name=dyn_bang_c_1, bone2Name=base_bang_c_1, swing1LimitDegrees=15, swing2LimitDegrees=10, twistLimitDegrees=0}
 {
  bone1Name=base_bang_l_1
  bone2Name=neck_02
  swing1LimitDegrees=45
  swing1Motion=Free
  swing2LimitDegrees=45
  swing2Motion=Free
  twistLimitDegrees=45
  twistMotion=Free
 }
 {bone1Name=dyn_bang_l_1, bone2Name=base_bang_l_1, swing1LimitDegrees=15, swing2LimitDegrees=10, twistLimitDegrees=0}
 {
  bone1Name=base_bang_r_1
  bone2Name=neck_02
  swing1LimitDegrees=45
  swing1Motion=Free
  swing2LimitDegrees=45
  swing2Motion=Free
  twistLimitDegrees=45
  twistMotion=Free
 }
 {bone1Name=dyn_bang_r_1, bone2Name=base_bang_r_1, swing1LimitDegrees=15, swing2LimitDegrees=10, twistLimitDegrees=0}
 {
  bone1Name=base_hair_A_l_1
  bone2Name=neck_02
  swing1LimitDegrees=45
  swing1Motion=Free
  swing2LimitDegrees=45
  swing2Motion=Free
  twistLimitDegrees=45
  twistMotion=Free
 }
 {bone1Name=dyn_hair_A_l_1, bone2Name=base_hair_A_l_1, swing1LimitDegrees=10, swing2LimitDegrees=5, twistLimitDegrees=0}
 {bone1Name=dyn_hair_A_l_2, bone2Name=dyn_hair_A_l_1, swing1LimitDegrees=15, swing2LimitDegrees=10, twistLimitDegrees=0}
 {
  bone1Name=base_hair_A_r_1
  bone2Name=neck_02
  swing1LimitDegrees=45
  swing1Motion=Free
  swing2LimitDegrees=45
  swing2Motion=Free
  twistLimitDegrees=45
  twistMotion=Free
 }
 {bone1Name=dyn_hair_A_r_1, bone2Name=base_hair_A_r_1, swing1LimitDegrees=10, swing2LimitDegrees=5, twistLimitDegrees=0}
 {bone1Name=dyn_hair_A_r_2, bone2Name=dyn_hair_A_r_1, swing1LimitDegrees=15, swing2LimitDegrees=10, twistLimitDegrees=0}
 {
  bone1Name=base_hair_B_l_1
  bone2Name=neck_02
  swing1LimitDegrees=45
  swing1Motion=Free
  swing2LimitDegrees=45
  swing2Motion=Free
  twistLimitDegrees=45
  twistMotion=Free
 }
 {bone1Name=dyn_hair_B_l_1, bone2Name=base_hair_B_l_1, swing1LimitDegrees=10, swing2LimitDegrees=5, twistLimitDegrees=0}
 {bone1Name=dyn_hair_B_l_2, bone2Name=dyn_hair_B_l_1, swing1LimitDegrees=15, swing2LimitDegrees=5, twistLimitDegrees=0}
 {
  bone1Name=base_hair_B_r_1
  bone2Name=neck_02
  swing1LimitDegrees=45
  swing1Motion=Free
  swing2LimitDegrees=45
  swing2Motion=Free
  twistLimitDegrees=45
  twistMotion=Free
 }
 {bone1Name=dyn_hair_B_r_1, bone2Name=base_hair_B_r_1, swing1LimitDegrees=10, swing2LimitDegrees=5, twistLimitDegrees=0}
 {bone1Name=dyn_hair_B_r_2, bone2Name=dyn_hair_B_r_1, swing1LimitDegrees=15, swing2LimitDegrees=10, twistLimitDegrees=0}
 {
  bone1Name=base_hair_Bk_l_1
  bone2Name=neck_02
  swing1LimitDegrees=45
  swing1Motion=Free
  swing2LimitDegrees=45
  swing2Motion=Free
  twistLimitDegrees=45
  twistMotion=Free
 }
 {bone1Name=dyn_hair_Bk_l_1, bone2Name=base_hair_Bk_l_1, swing1LimitDegrees=10, swing2LimitDegrees=5, twistLimitDegrees=0}
 {bone1Name=dyn_hair_Bk_l_2, bone2Name=dyn_hair_Bk_l_1, swing1LimitDegrees=15, swing2LimitDegrees=10, twistLimitDegrees=0}
 {
  bone1Name=base_hair_Bk_r_1
  bone2Name=neck_02
  swing1LimitDegrees=45
  swing1Motion=Free
  swing2LimitDegrees=45
  swing2Motion=Free
  twistLimitDegrees=45
  twistMotion=Free
 }
 {bone1Name=dyn_hair_Bk_r_1, bone2Name=base_hair_Bk_r_1, swing1LimitDegrees=10, swing2LimitDegrees=5, twistLimitDegrees=0}
 {bone1Name=dyn_hair_Bk_r_2, bone2Name=dyn_hair_Bk_r_1, swing1LimitDegrees=15, swing2LimitDegrees=10, twistLimitDegrees=0}
 {
  bone1Name=upperarm_l
  bone2Name=clavicle_l
  swing1LimitDegrees=45
  swing1Motion=Free
  swing2LimitDegrees=45
  swing2Motion=Free
  twistLimitDegrees=45
  twistMotion=Free
 }
 {
  bone1Name=upperarm_r
  bone2Name=clavicle_r
  swing1LimitDegrees=45
  swing1Motion=Free
  swing2LimitDegrees=45
  swing2Motion=Free
  twistLimitDegrees=45
  twistMotion=Free
 }
 {
  bone1Name=base_earring_l
  bone2Name=neck_02
  swing1LimitDegrees=45
  swing1Motion=Free
  swing2LimitDegrees=45
  swing2Motion=Free
  twistLimitDegrees=45
  twistMotion=Free
 }
 {bone1Name=dyn_earring_l, bone2Name=base_earring_l, swing1LimitDegrees=5, swing2LimitDegrees=8, twistLimitDegrees=45, twistMotion=Locked}
 {
  bone1Name=base_earring_r
  bone2Name=neck_02
  swing1LimitDegrees=45
  swing1Motion=Free
  swing2LimitDegrees=45
  swing2Motion=Free
  twistLimitDegrees=45
  twistMotion=Free
 }
 {bone1Name=dyn_earring_r, bone2Name=base_earring_r, swing1LimitDegrees=5, swing2LimitDegrees=8, twistLimitDegrees=45, twistMotion=Locked}
]
