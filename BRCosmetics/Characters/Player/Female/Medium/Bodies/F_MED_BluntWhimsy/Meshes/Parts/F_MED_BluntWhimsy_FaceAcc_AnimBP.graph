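# Animation Blueprint for the face accessory part of the F_MED_BluntWhimsy female character, adding rigged and physical motion.
# It copies the attached parent's pose, layering clavicle and arm branches from the owner's body mesh.
# A Control Rig adjusts that using smoothed crouch and targeting, pistol, skydiving and rotation inputs, then feeds RigidBody simulation in head-bone space.
# Skydiving events flip GravityOverride, but the RigidBody node shows fixed downward gravity.
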AnimBlueprint /BRCosmetics/Characters/Player/Female/Medium/Bodies/F_MED_BluntWhimsy/Meshes/Parts/F_MED_BluntWhimsy_FaceAcc_AnimBP
target_skeleton=&/F_MED_BluntWhimsy_FaceAcc_Skeleton
graphs [
 AnimGraph {nodes [
   n1 = Control Rig!(
    Source=Use cached pose 'Start_Cache'()
    Alpha
    pub isTargeting:double=Get isTargeting float() -> (isTargeting float:double)
    pub isUsingPistol:double=Get isUsingPistol() -> (isUsingPistol:double)
    pub isCrouching:double=Get isCrouching float() -> (isCrouching float:double)
    pub pawnDeltaRotation:Rotator=Get Pub Pawn Delta Rotation() -> (Pub Pawn Delta Rotation:Rotator)
    SkyDiving:double=Get isSkyDivingFloat() -> (isSkyDivingFloat:double)
   ) -> (Pose) {
    alphaBoolBlend {blendOption=Linear}
    controlRigAssetReference {blueprintRigClass=&/F_MED_BluntWhimsy_FaceAcc_CtrlRig.F_MED_BluntWhimsy_FaceAcc_CtrlRig_C}
   }
   n2 = RigidBody!(ComponentPose=Local To Component(LocalPose=@n1), Alpha, OverrideWorldGravity:Vector="0,0,-980", ExternalForce) -> (Pose) {
    bOverrideWorldGravity=true
    overridePhysicsAsset=&/F_MED_BluntWhimsy_FaceAcc_Physics
    simulationSpace=BaseBoneSpace
    alphaBoolBlend {blendOption=Linear}
    baseBoneRef {boneName=head}
    componentAppliedLinearAccClamp=<0 0 0>
    overrideWorldGravity {z=-980}
    simSpaceSettings {
     dampingAlpha=0
     maxAngularAcceleration=12
     maxAngularVelocity=3
     maxLinearAcceleration=1000
     maxLinearVelocity=300
     worldAlpha=1
     externalLinearDragV=<1 1 1>
    }
   }
   Output Pose(Result=Component To Local(ComponentPose=@n2)) {layerGroup=None}
   n3 = Layered blend per bone(
    BasePose=Copy Pose From Mesh() {bCopyCurves=true, bUseAttachedParent=true}
    BlendPoses_0=Copy Pose From Mesh(SourceMeshComponent=Get Body_Mesh())
   ) {layerSetup [{branchFilters [{boneName=clavicle_l}, {boneName=clavicle_r}, {boneName=upperarm_l}, {boneName=upperarm_r}]}]}
   Save cached pose 'Start_Cache'(Pose=@n3) {cacheName=Start_Cache}
  ]}
 EventGraph {nodes [
   Event BlueprintInitializeAnimation()
   > Set Body_Mesh(Body_Mesh=GetOwnerSkeletalMeshComponent())
   Event OnBeginSkydiving()
   > Set GravityOverride(GravityOverride="0,0,980")
   n1 = Event BlueprintUpdateAnimation()
   > Set Delta Time X(Delta Time X=@n1.DeltaTimeX)
   > n2 = Sequence()
   n3 = Get bIsCrouching()
   n4 = Get Delta Time X()
   Set isCrouching float(
    execute=@n2.then_1
    isCrouching float:double=FInterpTo(Current=Get isCrouching float(), Target="To Float (Boolean)"(InBool=@n3), DeltaTime=@n4, InterpSpeed=15)
   ) -> (Output_Get:double)
   n5 = Get bIsTargeting()
   > Set isTargeting float(isTargeting float:double=FInterpTo(Current=Get isTargeting float(), Target="To Float (Boolean)"(InBool=@n5), DeltaTime=@n4, InterpSpeed=15)) -> (Output_Get:double)
   n6 = Get bIsSkydiveDiveMode()
   n7 = Get bIsSkydiveFloating()
   > Set isSkyDivingFloat(isSkyDivingFloat:double="To Float (Boolean)"(InBool=OR Boolean(A=@n6, B=@n7))) -> (Output_Get:double)
   Event OnEndSkydiving()
   > Set GravityOverride(GravityOverride="0,0,-980")
   Event OnWeaponChanged()
   > Set isUsingPistol(isUsingPistol:double=0) -> (Output_Get:double)
   > n8 = Switch on EFortWeaponCoreAnimation!(execute, Selection:EFortWeaponCoreAnimation Enum=Get WeaponCoreAnim!() -> (WeaponCoreAnim)) -> (
    Melee:Exec
    Pistol:Exec
    Shotgun:Exec
    PaperBlueprint:Exec
    Rifle:Exec
    MeleeOneHand:Exec
    MachinePistol:Exec
    RocketLauncher:Exec
    GrenadeLauncher:Exec
    GoingCommando:Exec
    AssaultRifle:Exec
    TacticalShotgun:Exec
    SniperRifle:Exec
    TrapPlacement:Exec
    ShoulderLauncher:Exec
    AbilityDecoTool:Exec
    Crossbow:Exec
    C4:Exec
    RemoteControl:Exec
    DualWield:Exec
    AR_BullPup:Exec
    AR_ForwardGrip:Exec
    MedPackPaddles:Exec
    SMG_P90:Exec
    AR_DrumGun:Exec
    Consumable_Small:Exec
    Consumable_Large:Exec
    Balloon:Exec
    MountedTurret:Exec
    CreativeTool:Exec
    ExplosiveBow:Exec
    AshtonIndigo:Exec
    AshtonChicago:Exec
    MeleeDualWield:Exec
    Unarmed:Exec
   )
   Set isUsingPistol(execute=@n8.Pistol, isUsingPistol=1)
  ]}
]
variables [
 Body_Mesh {type=SkeletalMeshComponent}
 Delta Time X {type=number}
 GravityOverride {type=Vector, default=<0 0 -980>}
 isCrouching float {type=number}
 isSkyDivingFloat {type=number}
 isTargeting float {type=number}
 isUsingPistol {type=number}
 Pub Pawn Delta Rotation {type=Rotator}
]
